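; Building prop Blueprint for a cluster of office server racks that extends its building texture customization to three extra static meshes.
; The construction script calls the parent, then, if building texture data exists, applies the tint color and override material from an entry of that data to each secondary mesh.
; It also forces a generic server screen material onto slot 1 of those meshes and the main mesh.
; Event graph events are empty.
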(event EventBeginPlay)

(event Collision|EventActorBeginOverlap
  (params (OtherActor :type Actor%20Object%20Reference)))

(event EventTick
  (params (DeltaSeconds :type Float%20%28single-precision%29)))

(fn ConstructionScript
  (|CallParentFunction)
  (bind _outtexturedataarray (Building|GetBuildingTextureData))
  (if (> (Utilities|Array|Length _outtexturedataarray) 0)
    (for _array_element (Utilities|Array|MakeArray (Variables|Default|GetStaticMesh2) (Variables|Default|GetStaticMesh1) (Variables|Default|GetStaticMesh3))
      (Default|SetColoronSecondaryMesh (Math|Conversions|ToLinearColor(Color) (Variables|/Script/FortniteGame.BuildingTextureData|GetTintColor (Utilities|Array|Get(acopy) _outtexturedataarray))) _array_element)
      (Rendering|Material|SetMaterial _array_element 0 (Variables|/Script/FortniteGame.BuildingTextureData|GetOverrideMaterial (Utilities|Array|Get(acopy) _outtexturedataarray)))
      (Rendering|Material|SetMaterial _array_element 1 "/Game/Environments/Helios/Props/Office/Office_ServerRacks_A/Materials/MI_Office_ServerScreenGeneric_A.MI_Office_ServerScreenGeneric_A")
      (Rendering|Material|SetMaterial (Variables|Debug|GetStaticMeshComponent) 1 "/Game/Environments/Helios/Props/Office/Office_ServerRacks_A/Materials/MI_Office_ServerScreenGeneric_A.MI_Office_ServerScreenGeneric_A"))))
