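; Office server rack cluster prop Blueprint that configures an instanced server mesh, with no gameplay logic in its events.
; The construction script copies the main component's mesh to ISM_Server and applies the Server Material to slot 0.
; If building texture data exists, the first entry's tint is passed to SetColoronSecondaryMesh for that component.
; Exposed asset provider overrides add version 1 and collect the material property's references for allowing.
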(event EventBeginPlay)

(event Collision|EventActorBeginOverlap
  (params (OtherActor :type Actor%20Object%20Reference)))

(event EventTick
  (params (DeltaSeconds :type Float%20%28single-precision%29)))

(fn BuildExposedAssetProviderVersion(Implementation)
  (params (Version :type Fort%20Exposed%20Asset%20Provider%20Version%20Structure%20%28by%20ref%29))
  (|CallParentFunction Version)
  (ExposedAssetProvider|AddVersion Version 1)
  (return))

(fn CollectAssetReferencestoAllow(Implementation)
  (params (Collector :type Fort%20Exposed%20Asset%20Reference%20Collector%20Structure%20%28by%20ref%29))
  (|CallParentFunction Collector)
  (ExposedAssetProvider|CollectObjectPropertyAssetReferences Collector 0 (Variables|Default|GetServer%20Material))
  (return))

(fn ConstructionScript
  (bind _ism_server (Variables|Default|GetISM_Server))
  (|CallParentFunction)
  (Components|StaticMesh|SetStaticMesh _ism_server (Variables|/Script/Engine.StaticMeshComponent|GetStaticMesh (Variables|Debug|GetStaticMeshComponent)))
  (Rendering|Material|SetMaterial _ism_server 0 (Variables|Default|GetServer%20Material))
  (bind _outtexturedataarray (Building|GetBuildingTextureData))
  (if (> (Utilities|Array|Length _outtexturedataarray) 0)
    (Default|SetColoronSecondaryMesh (Math|Conversions|ToLinearColor(Color) (Variables|/Script/FortniteGame.BuildingTextureData|GetTintColor (Utilities|Array|Get(acopy) _outtexturedataarray))) _ism_server)))
